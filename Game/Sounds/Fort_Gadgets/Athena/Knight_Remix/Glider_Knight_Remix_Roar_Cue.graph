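# Sound cue for the Knight Remix glider's roar, layering a creature scream with a dragon fire burst.
# A random pick of three scream waves gets varied pitch and mixes with a quieter, heavily pitched-down fire burst that starts after a short fixed delay.
# It uses glider thrust attenuation and the glider land sound class, with a concurrency override that blocks new instances beyond a small limit.
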SoundCue /Game/Sounds/Fort_Gadgets/Athena/Knight_Remix/Glider_Knight_Remix_Roar_Cue
root=Mixer_0
nodes [
 {id=Delay_0, children [WavePlayer_0], properties {delayMax=0.5, delayMin=0.5}}
 {id=Mixer_0, children [Modulator_0,Modulator_1], properties {inputVolume [1,0.7]}}
 {id=Modulator_0, children [Random_0], properties {pitchMax=1.1, pitchMin=0.8}}
 {id=Modulator_1, children [Delay_0], properties {pitchMax=0.7, pitchMin=0.6}}
 {id=Random_0, children [WavePlayer_2,WavePlayer_3,WavePlayer_4], properties {weights [1,1,1]}}
 {id=WavePlayer_0, properties {soundWaveAssetPtr=/Game/Sounds/Fort_Gadgets/Dragon_Slash/Fire_of_the_Dragon_Burst_02}}
 {id=WavePlayer_2, properties {soundWaveAssetPtr=&/Glider_Knight_Remix_Scream_02}}
 {id=WavePlayer_3, properties {soundWaveAssetPtr=&/Glider_Knight_Remix_Scream_03}}
 {id=WavePlayer_4, properties {soundWaveAssetPtr=&/Glider_Knight_Remix_Scream_04}}
]
properties {
 attenuationSettings=&Fort_Player_Glider_Thrust_Athena_Att
 bOverrideConcurrency=true
 maxDistance=12500
 soundClassObject=&Fort_Gliders_Land
 volumeMultiplier=0.4
 concurrencyOverrides {maxCount=2, resolutionRule=PreventNew}
 concurrencySet [&Athena_GliderOpenClose_Conc]
}
